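# Base UI Kit layout widget for Fortnite frontend screens, with named slots for primary and secondary content, side nav, bottom bar, XP and overlays.
# PreConstruct runs Display, which collapses empty slots, sets background, panel widths and safe zone padding.
# A CVar decides whether the side nav overlays or pushes content by reparenting it.
# Construct keeps blurred or unblurred PostBufferUpdate widgets based on config; setters adjust compact style, padding and debug text.
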WidgetBlueprint /FortUILibrary/UIKit/Layout/WBP_UIKit_FrontendLayout
graphs [
 AutoCollapseSlot {nodes [
   n1 = AutoCollapseSlot!() -> (then, Slot:Named Slot Object Reference) {is=title K2Node_FunctionEntry}
   > SetVisibility(
    self=@n1
    InVisibility=Select(Option 0=Collapsed, Option 1=SelfHitTestInvisible, Index=HasAnyChildren!(self:Panel Widget Object Reference=@n1) -> (ReturnValue:bool))
   )
  ]}
 Display {nodes [
   Display() {is=title K2Node_FunctionEntry}
   > DisplayPostBuffer()
   > DisplayBackground()
   > DisplayContentPanels()
   > DisplaySideNav()
   > DisplaySlots()
   > DisplaySafezone()
  ]}
 DisplayBackground {nodes [
   DisplayBackground() {is=title K2Node_FunctionEntry}
   > n1 = Is Valid(InputObject=Get BackgroundMaterial() -> (BackgroundMaterial:Material Instance Object Reference))
   SetVisibility(execute=@n1.Is Not Valid, self=Get Img_Background() -> (Img_Background:Image Object Reference), InVisibility=Collapsed)
   SetBrushFromMaterial(execute=@n1.Is Valid, self=Get Img_Background(), Material=Get BackgroundMaterial())
   > n2 = GetDynamicMaterial(self=Get Img_Background())
   > Set MID_Background(MID_Background:Material Instance Dynamic Object Reference=@n2) -> (Output_Get:Material Instance Dynamic Object Reference)
   > SetVisibility(self=Get Img_Background(), InVisibility=Visible)
  ]}
 DisplayContentPanels {nodes [
   DisplayContentPanels() {is=title K2Node_FunctionEntry}
   > n1 = Cast To GridSlot!(
    execute
    Object=Get Slot!(self:Widget Object Reference=Get SizeBox_ContentPanel_Secondary() -> (SizeBox_ContentPanel_Secondary:Size Box Object Reference)) -> (Slot:Panel Slot Object Reference)
   ) -> (then, CastFailed, AsGrid Slot:Grid Slot Object Reference) {is=title K2Node_DynamicCast}
   > SetPadding!(
    execute
    self:Grid Slot Object Reference=@n1
    InPadding:Margin Structure=Get SecondaryContentPaddingOverride() -> (SecondaryContentPaddingOverride:Margin Structure)
   ) -> (then)
   > n2 = Cast To GridSlot(Object=Get Slot(self=Get SizeBox_ContentPanel_Primary() -> (SizeBox_ContentPanel_Primary:Size Box Object Reference)))
   > SetPadding(self=@n2, InPadding=Get PrimaryContentPaddingOverride() -> (PrimaryContentPaddingOverride:Margin Structure))
   > SetContentPanelWidthOverride!(
    execute
    self
    SizeBox:Size Box Object Reference=Get SizeBox_ContentPanel_Primary()
    SizeX:double=Get ContentPanelPrimary_SizeX() -> (ContentPanelPrimary_SizeX:double)
    Fill:double=Get ContentPanelPrimary_Fill() -> (ContentPanelPrimary_Fill:double)
   ) -> (then)
   > SetContentPanelWidthOverride(
    SizeBox=Get SizeBox_ContentPanel_Secondary()
    SizeX=Get ContentPanelSecondary_SizeX() -> (ContentPanelSecondary_SizeX:double)
    Fill=Get ContentPanelSecondary_Fill() -> (ContentPanelSecondary_Fill:double)
   )
   > AutoCollapseSlot!(
    execute
    self
    Slot:Named Slot Object Reference=Get Slot_ContentPanel_Secondary() -> (Slot_ContentPanel_Secondary:Named Slot Object Reference)
   ) -> (then) {is=title K2Node_CallFunction}
   > AutoCollapseSlot(Slot=Get Slot_ContentPanelPrimary() -> (Slot_ContentPanelPrimary:Named Slot Object Reference))
  ]}
 DisplayPostBuffer {nodes [
   n1 = Branch(Condition=Get bUsePostBufferUpdate() -> (bUsePostBufferUpdate:bool))
   > n2 = Create Widget!(
    execute:Exec
    Class:User Widget Class Reference=/FortUILibrary/Blueprints/PostProcessor/WBP_CaptureForPostBufferUpdate.WBP_CaptureForPostBufferUpdate_C
    OwningPlayer:Player Controller Object Reference
   ) -> (then:Exec, ReturnValue:WBP Capture For Post Buffer Update Object Reference)
   > AddChild(self=Get Overlay_PostProcessBuffer() -> (Overlay_PostProcessBuffer:Overlay Object Reference), Content=@n2)
   ClearChildren(execute=@n1.else, self=Get Overlay_PostProcessBuffer())
   DisplayPostBuffer() {is=title K2Node_FunctionEntry}
  ]}
 DisplaySafezone {nodes [
   DisplaySafezone() {is=title K2Node_FunctionEntry}
   > n1 = Cast To OverlaySlot!(execute, Object=Get Slot(self=Get SafeZone() -> (SafeZone:Safe Zone Object Reference))) -> (then, CastFailed, AsOverlay Slot:Overlay Slot Object Reference) {is=title K2Node_DynamicCast}
   n2 = Get!() -> (
    SafeZone_ContentPadding_Left:float=~
    SafeZone_ContentPadding_Top:float=~
    SafeZone_ContentPadding_Right:float=~
    SafeZone_ContentPadding_Bottom:float=~
   )
   > SetPadding!(
    execute
    self:Overlay Slot Object Reference=@n1
    InPadding_Left:float=@n2
    InPadding_Top:float=Select(
     Option 0:float=@n2.SafeZone_ContentPadding_Top
     Option 1:float=Get SafeZone_InGameTopPadding() -> (SafeZone_InGameTopPadding:double)
     Index=Get bUseInGamePadding() -> (bUseInGamePadding:bool)
    ) -> (ReturnValue:float)
    InPadding_Right:float=@n2.SafeZone_ContentPadding_Right
    InPadding_Bottom:float=@n2.SafeZone_ContentPadding_Bottom
   ) -> (then)
  ]}
 DisplaySideNav {nodes [
   DisplaySideNav() {is=title K2Node_FunctionEntry}
   > AutoCollapseSlot(Slot=Get Slot_SideNav_Filter() -> (Slot_SideNav_Filter:Named Slot Object Reference))
   > AutoCollapseSlot(Slot=Get Slot_SideNav_ClearBadges() -> (Slot_SideNav_ClearBadges:Named Slot Object Reference))
   > AutoCollapseSlot(Slot=Get Slot_SideNav() -> (Slot_SideNav:Named Slot Object Reference))
   n1 = GetNumericCVar!(CVar:String=UI.FrontendLayout.EnableSideNavOverlay) -> (OutValue:float, ReturnValue:bool)
   > n2 = Branch(Condition=@n1.ReturnValue)
   > n3 = Set OverlayNav(OverlayNav:bool="Equal (Float)"(A=@n1, B=1)) -> (Output_Get:bool)
   n4 = Cast To GridSlot(
    execute=[@n9.then, @n3.then]
    Object=Get Slot(self=Get Overlay_SideNav_PushContent() -> (Overlay_SideNav_PushContent:Overlay Object Reference))
   )
   n5 = float + float(A=Get CollapsedSideNav_SizeX() -> (CollapsedSideNav_SizeX:double), B=Get SideNav_PaddingRight() -> (SideNav_PaddingRight:double))
   n6 = Get OverlayNav() -> (OverlayNav:bool)
   n7 = Get SideNav_PaddingRight()
   > SetPadding(
    self:Grid Slot Object Reference=@n4
    InPadding_Left=0
    InPadding_Top=0
    InPadding_Right=Select(Option 0=0, Option 1=Select(Option 0=@n7, Option 1=@n5, Index=@n6), Index=GetSideNavHasContent!() -> (HasContent:bool))
    InPadding_Bottom=0
   )
   > n8 = Branch(Condition=GetSideNavHasContent())
   > AddChildToOverlay!(
    execute
    self:Overlay Object Reference=Select(
     Option 0:Overlay Object Reference=Get Overlay_SideNav_PushContent()
     Option 1:Overlay Object Reference=Get Overlay_SideNav_OverlayContent() -> (Overlay_SideNav_OverlayContent:Overlay Object Reference)
     Index=Get OverlayNav()
    ) -> (ReturnValue:Overlay Object Reference)
    Content:Widget Object Reference=Get VerticalBox_SideNavigation() -> (VerticalBox_SideNavigation:Vertical Box Object Reference)
   ) -> (then, ReturnValue:Overlay Slot Object Reference)
   > SetVisibility(self=[Get Overlay_SideNav_PushContent(), Get Overlay_SideNav_OverlayContent()], InVisibility=SelfHitTestInvisible)
   SetVisibility(execute=@n8.else, self=[Get Overlay_SideNav_PushContent(), Get Overlay_SideNav_OverlayContent()], InVisibility=Collapsed)
   n9 = Set OverlayNav(execute=@n2.else, OverlayNav=false)
  ]}
 DisplaySlots {nodes [
   DisplaySlots() {is=title K2Node_FunctionEntry}
   > AutoCollapseSlot(Slot=Get Slot_ContentOverlay() -> (Slot_ContentOverlay:Named Slot Object Reference))
   > AutoCollapseSlot(Slot=Get Slot_ScreenOverlay() -> (Slot_ScreenOverlay:Named Slot Object Reference))
   > AutoCollapseSlot(Slot=Get Slot_BottomBar() -> (Slot_BottomBar:Named Slot Object Reference))
   > AutoCollapseSlot(Slot=Get Slot_SideNav_Filter())
   > AutoCollapseSlot(Slot=Get Slot_SideNav())
   > AutoCollapseSlot(Slot=Get Slot_SideNav_ClearBadges())
   > AutoCollapseSlot(Slot=Get Slot_BackgroundOverlay() -> (Slot_BackgroundOverlay:Named Slot Object Reference))
   > AutoCollapseSlot(Slot=Get Slot_XP() -> (Slot_XP:Named Slot Object Reference))
  ]}
 EventGraph {nodes [
   Event PreConstruct()
   > Display() {is=title K2Node_CallFunction}
   Event Construct()
   n1 = GetConfigDataBool!(EntryName:String=EnableBlurForFrontEndWidgets) -> (bValueOut:bool, ReturnValue:bool)
   > n2 = Branch(Condition=Select(Option 0:bool=true, Option 1:bool=@n1, Index=@n1.ReturnValue) -> (ReturnValue:bool))
   > RemoveWidget!(
    execute
    Widget:Widget Object Reference=Get Overlay_PostBufferUpdate_BlurDisabled() -> (Overlay_PostBufferUpdate_BlurDisabled:Post Buffer Update Object Reference)
   ) -> (then)
   > RemoveWidget(Widget=Get Root_PostBufferUpdate_BlurDisabled() -> (Root_PostBufferUpdate_BlurDisabled:Post Buffer Update Object Reference))
   RemoveWidget(execute=@n2.else, Widget=Get Overlay_PostBufferUpdate() -> (Overlay_PostBufferUpdate:Post Buffer Update Object Reference))
   > RemoveWidget(Widget=Get Root_PostBufferUpdate() -> (Root_PostBufferUpdate:Post Buffer Update Object Reference))
  ]}
 GetSideNavHasContent {nodes [
   GetSideNavHasContent() {is=title K2Node_FunctionEntry}
   n1 = Get Slot_SideNav_ClearBadges()
   n2 = Get Slot_SideNav()
   n3 = Get Slot_SideNav_Filter()
   > Return Node!(
    execute
    HasContent:bool=OR Boolean!(A=HasAnyChildren(self=@n3), B=HasAnyChildren(self=@n2), C:bool=HasAnyChildren(self=@n1)) -> (ReturnValue)
   )
  ]}
 GetSideNavVisibility {nodes [
   GetSideNavVisibility() {is=title K2Node_FunctionEntry}
   > Return Node!(execute, IsVisible:bool=IsVisible(self=Get VerticalBox_SideNavigation()))
  ]}
 SetCompactUIStyle {nodes [
   n1 = SetCompactUIStyle!() -> (
    then
    ContentPanelSizeX:double=1128
    TopBarSpacer:double=152
    SideNavMaxSizeY:double=600
    SafeZoneMarginY:double=12
    SafeZoneMarginX:double=40
   )
   > Set ContentPanelPrimary_SizeX(ContentPanelPrimary_SizeX:double=@n1) -> (Output_Get:double)
   n2 = Get SafeZoneMarginX() -> (SafeZoneMarginX:double)
   > Set!(
    execute
    SafeZone_ContentPadding_Left:float=@n2
    SafeZone_ContentPadding_Top:float=Get TopBarSpacer() -> (TopBarSpacer:double)
    SafeZone_ContentPadding_Right:float=@n2
    SafeZone_ContentPadding_Bottom:float=Get SafeZoneMarginY() -> (SafeZoneMarginY:double)
   ) -> (then, Output_Get:Margin Structure)
   > Set SafeZone_InGameTopPadding(SafeZone_InGameTopPadding:double=Get SafeZoneMarginY()) -> (Output_Get:double)
   n3 = Get SideNavMaxSizeY() -> (SideNavMaxSizeY:double)
   > n4 = Branch(Condition="float > float"(A=@n3))
   n5 = Get SizeBox_SideNav() -> (SizeBox_SideNav:Size Box Object Reference)
   > n6 = SetMaxDesiredHeight!(execute, self:Size Box Object Reference=@n5, InMaxDesiredHeight:float=@n3) -> (then)
   SetContentPanelWidthOverride(execute=[@n6.then, @n7.then], SizeBox=Get SizeBox_ContentPanel_Primary(), SizeX=Get ContentPanelPrimary_SizeX(), Fill=0)
   > DisplaySafezone() {is=title K2Node_CallFunction}
   n7 = ClearMaxDesiredHeight(execute=@n4.else, self:Size Box Object Reference=@n5)
  ]}
 SetContentPanelSizeX {nodes [
   n1 = SetContentPanelSizeX!() -> (then, ContentPanelPrimarySizeX:double=~) {is=title K2Node_FunctionEntry}
   > Set ContentPanelPrimary_SizeX(ContentPanelPrimary_SizeX=@n1)
   > SetContentPanelWidthOverride(SizeBox=Get SizeBox_ContentPanel_Primary(), SizeX=Get ContentPanelPrimary_SizeX(), Fill=0)
  ]}
 SetContentPanelWidthOverride {nodes [
   SetContentPanelWidthOverride!() -> (then, SizeBox:Size Box Object Reference, SizeX:double=~, Fill:double=~) {is=title K2Node_FunctionEntry}
   > n1 = Cast To GridSlot(Object=Get Slot(self=Get SizeBox() -> (SizeBox:Size Box Object Reference)))
   > Set Column(Column:int=Get Column!(self:Grid Slot Object Reference=@n1) -> (Column:int=0)) -> (Output_Get:int=0)
   > n2 = Branch(Condition="float >= float"(A=Get SizeX() -> (SizeX:double)))
   > SetWidthOverride(self=Get SizeBox(), InWidthOverride=Get SizeX())
   > SetColumnFill!(
    execute
    self:Grid Panel Object Reference=Get GridContent() -> (GridContent:Grid Panel Object Reference)
    ColumnIndex:int=Get Column!() -> (Column=0)
    Coefficient:float=0
   ) -> (then)
   ClearWidthOverride(execute=@n2.else, self:Size Box Object Reference=Get SizeBox())
   > SetColumnFill(self=Get GridContent(), ColumnIndex=Get Column() -> (Column=0), Coefficient=Get Fill() -> (Fill:double))
  ]}
 SetDebugText {nodes [
   n1 = SetDebugText!() -> (then, InText:Text, isWarning:bool=~) {is=title K2Node_FunctionEntry}
   > n2 = Branch(Condition=TextIsEmpty(InText=@n1))
   > SetVisibility(self=Get Text_Debug() -> (Text_Debug:Common Text Object Reference), InVisibility=Collapsed)
   "SetText (Text)"(execute=@n2.else, self=Get Text_Debug(), InText=Get InText() -> (InText:Text))
   > SetDebugTextWarning!(execute, self, IsWarning:bool=Get isWarning() -> (isWarning:bool)) -> (then)
   > SetVisibility(self=Get Text_Debug(), InVisibility=SelfHitTestInvisible)
  ]}
 SetDebugTextWarning {nodes [
   n1 = SetDebugTextWarning!() -> (then, IsWarning:bool=~) {is=title K2Node_FunctionEntry}
   > SetColorAndOpacity(
    self=Get Text_Debug()
    InColorAndOpacity=Make Slate Color(SpecifiedColor=Select(Option 0:Linear Color Structure=(R=1,G=1,B=1,A=1), Option 1:Linear Color Structure=(R=1,G=0,B=0,A=1), Index=@n1) -> (ReturnValue:Linear Color Structure))
   )
  ]}
 SetSafezoneTopPadding {nodes [
   n1 = SetSafezoneTopPadding!() -> (then, Safezone_FrontendTopPadding:double=136, SafeZone_InGameTopPadding:double=32, bUseCollapsedSafezoneTopPadding:bool=~) {is=title K2Node_FunctionEntry}
   > Set SafeZone_InGameTopPadding(SafeZone_InGameTopPadding=@n1.SafeZone_InGameTopPadding)
   > Set bUseInGamePadding(bUseInGamePadding:bool=@n1.bUseCollapsedSafezoneTopPadding) -> (Output_Get:bool)
   n2 = Get() -> (SafeZone_ContentPadding_Left=~, SafeZone_ContentPadding_Top=~, SafeZone_ContentPadding_Right=~, SafeZone_ContentPadding_Bottom=~)
   > Set(
    SafeZone_ContentPadding_Left=@n2
    SafeZone_ContentPadding_Top=Get Safezone_FrontendTopPadding() -> (Safezone_FrontendTopPadding:double)
    SafeZone_ContentPadding_Right=@n2.SafeZone_ContentPadding_Right
    SafeZone_ContentPadding_Bottom=@n2.SafeZone_ContentPadding_Bottom
   )
   > DisplaySafezone()
  ]}
 SetSideNavVisibility {nodes [
   n1 = SetSideNavVisibility!() -> (then, IsVisible:bool=~) {is=title K2Node_FunctionEntry}
   > SetVisibility(
    self=Get VerticalBox_SideNavigation()
    InVisibility=Select(Option 0:ESlateVisibility Enum=Collapsed, Option 1:ESlateVisibility Enum=SelfHitTestInvisible, Index=@n1) -> (ReturnValue:ESlateVisibility Enum)
   )
  ]}
 SetUseCollapseTopPadding {nodes [
   n1 = SetUseCollapseTopPadding!() -> (then, bUseCollapseTopBarPadding:bool=~) {is=title K2Node_FunctionEntry}
   > Set bUseInGamePadding(bUseInGamePadding=@n1)
   > DisplaySafezone()
  ]}
]
variables [
 BackgroundMaterial {type=MaterialInstance, default=null}
 bUseInGamePadding {type=boolean}
 bUsePostBufferUpdate {type=boolean}
 CollapsedSideNav_SizeX {type=number, default=100}
 ContentPanelPrimary_Fill {type=number, default=1}
 ContentPanelPrimary_SizeX {type=number, default=1008}
 ContentPanelSecondary_Fill {type=number, default=1}
 ContentPanelSecondary_SizeX {type=number, default=-1}
 MID_Background {type=MaterialInstanceDynamic}
 PrimaryContentPaddingOverride {type=Margin, default=<ltrb 0 0 0 0>}
 SafeZone_ContentPadding {type=Margin, default=<ltrb 80 136 80 32>}
 SafeZone_InGameTopPadding {type=number, default=32}
 SecondaryContentPaddingOverride {type=Margin, default=<ltrb 0 0 0 0>}
 SideNav_PaddingRight {type=number, default=34}
]
widgets {
 info {inheritedWidgetCount=0, parentClass=/Script/FortniteUI.UIKitFrontendLayoutWidget, rootWidgetClass=&Overlay, widgetCount=29}
 widgets [
  Overlay_34 {widgetClassPath=&Overlay, widget=Overlay_34}
  Img_Background {
   widgetClassPath=&Image
   bIsVariable=true
   parent=Overlay_34
   widget=Img_Background
   properties {brush {
     bIsDynamicallyLoaded=false
     drawAs=Image
     imageType=FullColor
     mirroring=NoMirror
     resourceName=None
     resourceObject=null
     tiling=NoTile
     imageSize=<0 0>
     margin=<ltrb 0 0 0 0>
     outlineSettings {
      bUseBrushTransparency=false
      roundingType=HalfHeightRadius
      width=0
      color {colorUseRule=UseColor_Specified, specifiedColor=<rgba 0 0 0 0>}
      cornerRadii=<0 0 0 0>
     }
     tintColor {colorUseRule=UseColor_Specified, specifiedColor=<rgba 1 1 1 1>}
     uVRegion {bIsValid=false, max=<0 0>, min=<0 0>}
    }}
   slot {class=OverlaySlot, properties {horizontalAlignment=HAlign_Fill, verticalAlignment=VAlign_Fill}}
  }
  Slot_BackgroundOverlay {
   widgetClassPath=&NamedSlot
   bIsVariable=true
   parent=Overlay_34
   widget=Slot_BackgroundOverlay
   slot {class=OverlaySlot, properties {horizontalAlignment=HAlign_Fill, verticalAlignment=VAlign_Fill}}
  }
  Root_PostBufferUpdate {
   widgetClassPath=/Script/UMG.PostBufferUpdate
   bIsVariable=true
   parent=Overlay_34
   widget=Root_PostBufferUpdate
   properties {
    visibility=HitTestInvisible
    updateBufferInfos [
     {
      bufferToUpdate=ESlatePostRT_0
      postParamUpdater=&self:WidgetTree.Root_PostBufferUpdate.PostBufferBlurUpdater_0
     }
     {bufferToUpdate=ESlatePostRT_1, postParamUpdater=null}
    ]
   }
   slot {class=OverlaySlot, properties {horizontalAlignment=HAlign_Fill, verticalAlignment=VAlign_Fill}}
  }
  Root_PostBufferUpdate_BlurDisabled {
   widgetClassPath=&PostBufferUpdate
   bIsVariable=true
   parent=Overlay_34
   widget=Root_PostBufferUpdate_BlurDisabled
   properties {
    visibility=HitTestInvisible
    updateBufferInfos [{bufferToUpdate=ESlatePostRT_0, postParamUpdater=null}, {bufferToUpdate=ESlatePostRT_1, postParamUpdater=null}]
   }
   slot {class=OverlaySlot, properties {horizontalAlignment=HAlign_Fill, verticalAlignment=VAlign_Fill}}
  }
  SafeZone {
   widgetClassPath=&SafeZone
   bIsVariable=true
   parent=Overlay_34
   widget=SafeZone
   slot {class=OverlaySlot, properties {horizontalAlignment=HAlign_Fill, verticalAlignment=VAlign_Fill}}
  }
  Overlay_52 {widgetClassPath=&Overlay, parent=SafeZone, widget=Overlay_52, slot {class=SafeZoneSlot}}
  SizeBox_Content {
   widgetClassPath=&SizeBox
   bIsVariable=true
   parent=Overlay_52
   widget=SizeBox_Content
   properties {bOverride_WidthOverride=true, widthOverride=1920}
   slot {class=OverlaySlot, properties {horizontalAlignment=HAlign_Center, verticalAlignment=VAlign_Fill}}
  }
  Overlay_Content {widgetClassPath=&Overlay, parent=SizeBox_Content, widget=Overlay_Content, slot {class=SizeBoxSlot}}
  Text_Debug {
   widgetClassPath=&CommonTextBlock
   bIsVariable=true
   parent=Overlay_Content
   widget=Text_Debug
   properties {
    bAutoCollapseWithEmptyText=true
    colorAndOpacity {colorUseRule=UseColor_Specified, specifiedColor=<rgba 0.5776 0.5776 0.5776 1>}
    font {
     bForceMonospaced=false
     bMaterialIsStencil=false
     fontMaterial=null
     fontObject=&HeadingNow
     letterSpacing=0
     monospacedWidth=1
     size=10.5
     skewAmount=0
     typefaceFontName=74 Regular
     outlineSettings {
      bApplyOutlineToDropShadows=false
      bMiteredCorners=false
      bSeparateFillAlpha=false
      outlineMaterial=null
      outlineSize=0
      outlineColor=<rgba 0 0 0 1>
     }
    }
    shadowOffset=<0 0>
   }
   slot {class=OverlaySlot, properties {horizontalAlignment=HAlign_Fill, padding=<ltrb 0 -35 0 5>}}
  }
  GridContent {
   widgetClassPath=&GridPanel
   bIsVariable=true
   parent=Overlay_Content
   widget=GridContent
   properties {columnFill [0,0,1], rowFill [1]}
   slot {class=OverlaySlot, properties {horizontalAlignment=HAlign_Fill, verticalAlignment=VAlign_Fill}}
  }
  Overlay_SideNav_PushContent {
   widgetClassPath=&Overlay
   bIsVariable=true
   parent=GridContent
   widget=Overlay_SideNav_PushContent
   slot {class=GridSlot, properties {layer=3, padding=<ltrb 0 0 134 0>}}
  }
  SizeBox_ContentPanel_Primary {
   widgetClassPath=&SizeBox
   bIsVariable=true
   parent=GridContent
   widget=SizeBox_ContentPanel_Primary
   properties {
    bOverride_WidthOverride=true
    clipping=ClipToBounds
    navigation=&self:WidgetTree.SizeBox_ContentPanel_Primary.WidgetNavigation_0
    widthOverride=1008
   }
   slot {class=GridSlot, properties {column=1, layer=2}}
  }
  Slot_ContentPanelPrimary {
   widgetClassPath=&NamedSlot
   bIsVariable=true
   parent=SizeBox_ContentPanel_Primary
   widget=Slot_ContentPanelPrimary
   slot {class=SizeBoxSlot}
  }
  Overlay_PostProcessBuffer {
   widgetClassPath=&Overlay
   bIsVariable=true
   parent=GridContent
   widget=Overlay_PostProcessBuffer
   slot {class=GridSlot, properties {columnSpan=3, layer=1}}
  }
  Overlay_PostBufferUpdate {
   widgetClassPath=&PostBufferUpdate
   bIsVariable=true
   parent=Overlay_PostProcessBuffer
   widget=Overlay_PostBufferUpdate
   properties {
    visibility=HitTestInvisible
    updateBufferInfos [
     {
      bufferToUpdate=ESlatePostRT_0
      postParamUpdater=&self:WidgetTree.Overlay_PostBufferUpdate.PostBufferBlurUpdater_1
     }
     {bufferToUpdate=ESlatePostRT_1, postParamUpdater=null}
    ]
   }
   slot {class=OverlaySlot, properties {horizontalAlignment=HAlign_Fill, verticalAlignment=VAlign_Fill}}
  }
  Overlay_PostBufferUpdate_BlurDisabled {
   widgetClassPath=&PostBufferUpdate
   bIsVariable=true
   parent=Overlay_PostProcessBuffer
   widget=Overlay_PostBufferUpdate_BlurDisabled
   properties {
    visibility=HitTestInvisible
    updateBufferInfos [{bufferToUpdate=ESlatePostRT_0, postParamUpdater=null}, {bufferToUpdate=ESlatePostRT_1, postParamUpdater=null}]
   }
   slot {class=OverlaySlot, properties {horizontalAlignment=HAlign_Fill, verticalAlignment=VAlign_Fill}}
  }
  SizeBox_ContentPanel_Secondary {
   widgetClassPath=&SizeBox
   bIsVariable=true
   parent=GridContent
   widget=SizeBox_ContentPanel_Secondary
   properties {
    navigation=&self:WidgetTree.SizeBox_ContentPanel_Secondary.WidgetNavigation_0
    widthOverride=1008
   }
   slot {class=GridSlot, properties {column=2}}
  }
  Slot_ContentPanel_Secondary {
   widgetClassPath=&NamedSlot
   bIsVariable=true
   parent=SizeBox_ContentPanel_Secondary
   widget=Slot_ContentPanel_Secondary
   slot {class=SizeBoxSlot}
  }
  Overlay_SideNav_OverlayContent {
   widgetClassPath=&Overlay
   bIsVariable=true
   parent=Overlay_Content
   widget=Overlay_SideNav_OverlayContent
   slot {class=OverlaySlot, properties {horizontalAlignment=HAlign_Fill, verticalAlignment=VAlign_Fill}}
  }
  VerticalBox_SideNavigation {
   widgetClassPath=&VerticalBox
   bIsVariable=true
   parent=Overlay_SideNav_OverlayContent
   widget=VerticalBox_SideNavigation
   slot {class=OverlaySlot}
  }
  Slot_SideNav_Filter {
   widgetClassPath=&NamedSlot
   bIsVariable=true
   parent=VerticalBox_SideNavigation
   widget=Slot_SideNav_Filter
   slot {class=VerticalBoxSlot}
  }
  SizeBox_SideNav {
   widgetClassPath=&SizeBox
   bIsVariable=true
   parent=VerticalBox_SideNavigation
   widget=SizeBox_SideNav
   properties {maxDesiredHeight=600, maxDesiredWidth=250}
   slot {class=VerticalBoxSlot, properties {size {sizeRule=Fill, value=1}}}
  }
  Slot_SideNav {widgetClassPath=&NamedSlot, bIsVariable=true, parent=SizeBox_SideNav, widget=Slot_SideNav, slot {class=SizeBoxSlot}}
  Slot_SideNav_ClearBadges {
   widgetClassPath=&NamedSlot
   bIsVariable=true
   parent=VerticalBox_SideNavigation
   widget=Slot_SideNav_ClearBadges
   slot {class=VerticalBoxSlot}
  }
  Slot_ContentOverlay {
   widgetClassPath=&NamedSlot
   bIsVariable=true
   parent=Overlay_Content
   widget=Slot_ContentOverlay
   slot {class=OverlaySlot, properties {horizontalAlignment=HAlign_Fill, verticalAlignment=VAlign_Fill}}
  }
  Slot_BottomBar {
   widgetClassPath=&NamedSlot
   bIsVariable=true
   parent=Overlay_Content
   widget=Slot_BottomBar
   slot {class=OverlaySlot, properties {horizontalAlignment=HAlign_Right, verticalAlignment=VAlign_Bottom}}
  }
  Slot_XP {
   widgetClassPath=&NamedSlot
   bIsVariable=true
   parent=Overlay_52
   widget=Slot_XP
   slot {class=OverlaySlot, properties {horizontalAlignment=HAlign_Right}}
  }
  Slot_ScreenOverlay {
   widgetClassPath=&NamedSlot
   bIsVariable=true
   parent=Overlay_34
   widget=Slot_ScreenOverlay
   slot {class=OverlaySlot, properties {horizontalAlignment=HAlign_Fill, verticalAlignment=VAlign_Fill}}
  }
 ]
}
