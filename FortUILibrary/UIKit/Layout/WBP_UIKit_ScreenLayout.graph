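# UI Kit frontend screen layout with named slots for background, side nav, primary/secondary content, overlays and bottom bar.
# PreConstruct sizes grid columns by fixed width or fill, sets safe-zone padding for compact UI and optional top nav, and collapses empty slots.
# The CVar UI.FrontendLayout.EnableSideNavOverlay reparents the side nav between a content-pushing grid column and an overlay, which reserves only collapsed width plus padding.
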WidgetBlueprint /FortUILibrary/UIKit/Layout/WBP_UIKit_ScreenLayout
graphs [
 AutoCollapseSlot {nodes [
   n1 = AutoCollapseSlot!() -> (then, Slot:Named Slot Object Reference) {is=title K2Node_FunctionEntry}
   > SetVisibility(
    self=@n1
    InVisibility=Select(Option 0=Collapsed, Option 1=SelfHitTestInvisible, Index=HasAnyChildren!(self:Panel Widget Object Reference=@n1) -> (ReturnValue:bool))
   )
  ]}
 CollapseUnusedSlots {nodes [
   CollapseUnusedSlots() {is=title K2Node_FunctionEntry}
   > n1 = For Each Loop(Array:Array of Named Slot Object References=Make Array!(
     "[0]":Named Slot Object Reference=Get Slot_SideNav() -> (Slot_SideNav:Named Slot Object Reference)
     "[1]":Named Slot Object Reference=Get Slot_ScreenOverlay() -> (Slot_ScreenOverlay:Named Slot Object Reference)
     "[2]":Named Slot Object Reference=Get Slot_PrimaryContent() -> (Slot_PrimaryContent:Named Slot Object Reference)
     "[3]":Named Slot Object Reference=Get Slot_SecondaryContent() -> (Slot_SecondaryContent:Named Slot Object Reference)
     "[4]":Named Slot Object Reference=Get Slot_ContentOverlay() -> (Slot_ContentOverlay:Named Slot Object Reference)
     "[5]":Named Slot Object Reference=Get Slot_BottomBar() -> (Slot_BottomBar:Named Slot Object Reference)
     "[6]":Named Slot Object Reference=Get Slot_Background() -> (Slot_Background:Named Slot Object Reference)
    ) -> (Array:Array of Named Slot Object References)) -> (Array Element:Named Slot Object Reference)
   > AutoCollapseSlot!(execute, self, Slot:Named Slot Object Reference=@n1) -> (then) {is=title K2Node_CallFunction}
  ]}
 Display {nodes [
   Display() {is=title K2Node_FunctionEntry}
   > DisplayContentPanels()
   > DisplaySideNav()
   > DisplaySafeZone()
   > CollapseUnusedSlots() {is=title K2Node_CallFunction}
  ]}
 DisplayContentPanel {nodes [
   n1 = DisplayContentPanel!() -> (then, SizeBox:Size Box Object Reference, Padding:Margin Structure, SizeX:double=~, Fill:double=~) {is=title K2Node_FunctionEntry}
   n2 = SlotAsGridSlot(Widget=@n1)
   > SetPadding!(execute, self:Grid Slot Object Reference=@n2, InPadding:Margin Structure=@n1.Padding) -> (then)
   > Set Column(Column:int=Get Column!(self:Grid Slot Object Reference=@n2) -> (Column:int=0)) -> (Output_Get:int=0)
   > n3 = Branch(Condition="float >= float"(A=Get SizeX() -> (SizeX:double)))
   > SetWidthOverride(self=Get SizeBox() -> (SizeBox:Size Box Object Reference), InWidthOverride=Get SizeX())
   > SetColumnFill!(
    execute
    self:Grid Panel Object Reference=Get GridContent() -> (GridContent:Grid Panel Object Reference)
    ColumnIndex:int=Get Column!() -> (Column=0)
    Coefficient:float=0
   ) -> (then)
   ClearWidthOverride(execute=@n3.else, self:Size Box Object Reference=Get SizeBox())
   > SetColumnFill(self=Get GridContent(), ColumnIndex=Get Column() -> (Column=0), Coefficient=Get Fill() -> (Fill:double))
  ]}
 DisplayContentPanels {nodes [
   DisplayContentPanels() {is=title K2Node_FunctionEntry}
   > DisplayContentPanel!(
    execute
    self
    SizeBox:Size Box Object Reference=Get SizeBox_ContentPanel_Primary() -> (SizeBox_ContentPanel_Primary:Size Box Object Reference)
    Padding:Margin Structure=Get PrimaryContent_Padding() -> (PrimaryContent_Padding:Margin Structure)
    SizeX:double=Get PrimaryContent_SizeX() -> (PrimaryContent_SizeX:double)
    Fill:double=Get PrimaryContent_Fill() -> (PrimaryContent_Fill:double)
   ) -> (then) {is=title K2Node_CallFunction}
   > DisplayContentPanel(
    SizeBox=Get SizeBox_SecondaryContent() -> (SizeBox_SecondaryContent:Size Box Object Reference)
    Padding=Get SecondaryContent_Padding() -> (SecondaryContent_Padding:Margin Structure)
    SizeX=Get SecondaryContent_SizeX() -> (SecondaryContent_SizeX:double)
    Fill=Get SecondaryContent_Fill() -> (SecondaryContent_Fill:double)
   )
  ]}
 DisplaySafeZone {nodes [
   DisplaySafeZone() {is=title K2Node_FunctionEntry}
   n1 = IsUsingCompactUI()
   n2 = Break Margin!(Margin:"Margin Structure (by ref)"=Select(
     Option 0:Margin Structure=Get SafeZone_ContentPadding() -> (SafeZone_ContentPadding:Margin Structure)
     Option 1:Margin Structure=Get Mobile_SafeZone_ContentPadding() -> (Mobile_SafeZone_ContentPadding:Margin Structure)
     Index=@n1
    ) -> (ReturnValue:Margin Structure)) -> (Left:float=~, Top:float=~, Right:float=~, Bottom:float=~) {is=title K2Node_BreakStruct}
   n3 = Get SafeZone_TopNavPadding() -> (SafeZone_TopNavPadding:double)
   n4 = Get Mobile_SafeZone_TopNavPadding() -> (Mobile_SafeZone_TopNavPadding:double)
   > SetPadding!(
    execute
    self:Overlay Slot Object Reference=SlotAsOverlaySlot(Widget=Get SafeZone() -> (SafeZone:Safe Zone Object Reference))
    InPadding_Left:float=@n2
    InPadding_Top:float=Select(
     Option 0:float=@n2.Top
     Option 1:float=Select(Option 0:double=@n3, Option 1:double=@n4, Index=@n1) -> (ReturnValue:double)
     Index=Get bUseTopNavPadding() -> (bUseTopNavPadding:bool)
    ) -> (ReturnValue:float)
    InPadding_Right:float=@n2.Right
    InPadding_Bottom:float=@n2.Bottom
   ) -> (then)
   n5 = Get bApplySafeZone() -> (bApplySafeZone:bool)
   > SetSidesToPad!(execute, self:Safe Zone Object Reference=Get SafeZone(), InPadLeft:bool=@n5, InPadRight:bool=@n5, InPadTop:bool=@n5, InPadBottom:bool=@n5) -> (then)
  ]}
 DisplaySideNav {nodes [
   DisplaySideNav() {is=title K2Node_FunctionEntry}
   n1 = GetNumericCVar!(CVar:String=UI.FrontendLayout.EnableSideNavOverlay) -> (OutValue:float, ReturnValue:bool)
   > n2 = Branch(Condition=@n1.ReturnValue)
   > n3 = Set OverlayNav(OverlayNav:bool="Equal (Float)"(A=@n1, B=1)) -> (Output_Get:bool)
   n4 = float + float(A=Get SideNav_CollapsedSizeX() -> (SideNav_CollapsedSizeX:double), B=Get SideNav_PaddingRight() -> (SideNav_PaddingRight:double))
   n5 = Get OverlayNav() -> (OverlayNav:bool)
   n6 = Get SideNav_PaddingRight()
   SetPadding(
    execute=[@n8.then, @n3.then]
    self:Grid Slot Object Reference=SlotAsGridSlot(Widget=Get Overlay_SideNav_PushContent() -> (Overlay_SideNav_PushContent:Overlay Object Reference))
    InPadding_Left=0
    InPadding_Top=0
    InPadding_Right=Select(Option 0:float=0, Option 1:float=Select(Option 0=@n6, Option 1=@n4, Index=@n5), Index=GetSideNavHasContent!(self) -> (HasContent:bool)) -> (ReturnValue:float)
    InPadding_Bottom=0
   )
   > n7 = Branch(Condition=GetSideNavHasContent())
   > AddChildToOverlay!(
    execute
    self:Overlay Object Reference=Select(
     Option 0:Overlay Object Reference=Get Overlay_SideNav_PushContent()
     Option 1:Overlay Object Reference=Get Overlay_SideNav_OverlayContent() -> (Overlay_SideNav_OverlayContent:Overlay Object Reference)
     Index=Get OverlayNav()
    ) -> (ReturnValue:Overlay Object Reference)
    Content:Widget Object Reference=Get VerticalBox_SideNavigation() -> (VerticalBox_SideNavigation:Vertical Box Object Reference)
   ) -> (then, ReturnValue:Overlay Slot Object Reference)
   > SetVisibility(self=[Get Overlay_SideNav_PushContent(), Get Overlay_SideNav_OverlayContent()], InVisibility=SelfHitTestInvisible)
   SetVisibility(execute=@n7.else, self=[Get Overlay_SideNav_PushContent(), Get Overlay_SideNav_OverlayContent()], InVisibility=Collapsed)
   n8 = Set OverlayNav(execute=@n2.else, OverlayNav=false)
  ]}
 EventGraph {nodes [
   Event PreConstruct()
   > Display() {is=title K2Node_CallFunction}
  ]}
 GetSideNavHasContent {nodes [
   GetSideNavHasContent() {is=title K2Node_FunctionEntry}
   > Return Node!(execute, HasContent:bool=HasAnyChildren(self=Get Slot_SideNav()))
  ]}
 GetSideNavIsVisible {nodes [
   GetSideNavIsVisible() {is=title K2Node_FunctionEntry}
   > Return Node!(execute, IsVisible:bool=IsVisible(self=Get VerticalBox_SideNavigation()))
  ]}
 SetUseTopNavPadding {nodes [
   n1 = SetUseTopNavPadding!() -> (then, bUseTopNavPadding:bool=~) {is=title K2Node_FunctionEntry}
   > Set bUseTopNavPadding(bUseTopNavPadding:bool=@n1) -> (Output_Get:bool)
   > DisplaySafeZone() {is=title K2Node_CallFunction}
  ]}
]
variables [
 bApplySafeZone {type=boolean, default=true}
 bUseTopNavPadding {type=boolean, default=true}
 Mobile_SafeZone_ContentPadding {type=Margin, default=<ltrb 48 24 48 24>}
 Mobile_SafeZone_TopNavPadding {type=number, default=152}
 PrimaryContent_Fill {type=number, default=1}
 PrimaryContent_Padding {type=Margin, default=<ltrb 0 0 0 0>}
 PrimaryContent_SizeX {type=number, default=-1}
 SafeZone_ContentPadding {type=Margin, default=<ltrb 80 32 80 32>}
 SafeZone_TopNavPadding {type=number, default=136}
 SecondaryContent_Fill {type=number, default=1}
 SecondaryContent_Padding {type=Margin, default=<ltrb 0 0 0 0>}
 SecondaryContent_SizeX {type=number, default=-1}
 SideNav_CollapsedSizeX {type=number, default=100}
 SideNav_PaddingRight {type=number, default=34}
]
widgets {
 info {
  inheritedWidgetCount=0
  namedSlotCount=1
  parentClass=/Script/FortniteUI.UIKitFrontendLayoutWidget
  rootWidgetClass=&Overlay
  widgetCount=21
 }
 widgets [
  Overlay_34 {widgetClassPath=&Overlay, widget=Overlay_34}
  Slot_Background {
   widgetClassPath=&NamedSlot
   bIsVariable=true
   parent=Overlay_34
   widget=Slot_Background
   slot {class=OverlaySlot, properties {horizontalAlignment=HAlign_Fill, verticalAlignment=VAlign_Fill}}
  }
  SafeZone {
   widgetClassPath=&SafeZone
   bIsVariable=true
   parent=Overlay_34
   widget=SafeZone
   slot {class=OverlaySlot, properties {horizontalAlignment=HAlign_Fill, verticalAlignment=VAlign_Fill}}
  }
  Overlay_52 {widgetClassPath=&Overlay, parent=SafeZone, widget=Overlay_52, slot {class=SafeZoneSlot}}
  SizeBox_Content {
   widgetClassPath=&SizeBox
   bIsVariable=true
   parent=Overlay_52
   widget=SizeBox_Content
   properties {bOverride_WidthOverride=true, widthOverride=1920}
   slot {class=OverlaySlot, properties {horizontalAlignment=HAlign_Fill, verticalAlignment=VAlign_Fill}}
  }
  Overlay_Content {widgetClassPath=&Overlay, parent=SizeBox_Content, widget=Overlay_Content, slot {class=SizeBoxSlot}}
  GridContent {
   widgetClassPath=&GridPanel
   bIsVariable=true
   parent=Overlay_Content
   widget=GridContent
   properties {columnFill [0,0,1], rowFill [1]}
   slot {class=OverlaySlot, properties {horizontalAlignment=HAlign_Fill, verticalAlignment=VAlign_Fill}}
  }
  Overlay_SideNav_PushContent {
   widgetClassPath=&Overlay
   bIsVariable=true
   parent=GridContent
   widget=Overlay_SideNav_PushContent
   slot {class=GridSlot, properties {layer=2, padding=<ltrb 0 0 134 0>}}
  }
  Overlay_PostProcess {
   widgetClassPath=&Overlay
   parent=GridContent
   widget=Overlay_PostProcess
   slot {class=GridSlot, properties {columnSpan=3, layer=1}}
  }
  SizeBox_ContentPanel_Primary {
   widgetClassPath=&SizeBox
   bIsVariable=true
   parent=GridContent
   widget=SizeBox_ContentPanel_Primary
   properties {bOverride_WidthOverride=true, widthOverride=1008}
   slot {class=GridSlot, properties {column=1, layer=2}}
  }
  Slot_PrimaryContent {
   widgetClassPath=&NamedSlot
   bIsVariable=true
   parent=SizeBox_ContentPanel_Primary
   widget=Slot_PrimaryContent
   slot {class=SizeBoxSlot}
  }
  SizeBox_SecondaryContent {
   widgetClassPath=&SizeBox
   bIsVariable=true
   parent=GridContent
   widget=SizeBox_SecondaryContent
   properties {widthOverride=1008}
   slot {class=GridSlot, properties {column=2}}
  }
  Slot_SecondaryContent {
   widgetClassPath=&NamedSlot
   bIsVariable=true
   parent=SizeBox_SecondaryContent
   widget=Slot_SecondaryContent
   slot {class=SizeBoxSlot}
  }
  Overlay_SideNav_OverlayContent {
   widgetClassPath=&Overlay
   bIsVariable=true
   parent=Overlay_Content
   widget=Overlay_SideNav_OverlayContent
   slot {class=OverlaySlot, properties {horizontalAlignment=HAlign_Fill, verticalAlignment=VAlign_Fill}}
  }
  VerticalBox_SideNavigation {
   widgetClassPath=&VerticalBox
   bIsVariable=true
   parent=Overlay_SideNav_OverlayContent
   widget=VerticalBox_SideNavigation
   slot {class=OverlaySlot}
  }
  SizeBox_SideNav {
   widgetClassPath=&SizeBox
   bIsVariable=true
   parent=VerticalBox_SideNavigation
   widget=SizeBox_SideNav
   properties {maxDesiredHeight=600, maxDesiredWidth=250}
   slot {class=VerticalBoxSlot, properties {size {sizeRule=Fill, value=1}}}
  }
  Slot_SideNav {widgetClassPath=&NamedSlot, bIsVariable=true, parent=SizeBox_SideNav, widget=Slot_SideNav, slot {class=SizeBoxSlot}}
  Slot_ContentOverlay {
   widgetClassPath=&NamedSlot
   bIsVariable=true
   parent=Overlay_Content
   widget=Slot_ContentOverlay
   slot {class=OverlaySlot, properties {horizontalAlignment=HAlign_Fill, verticalAlignment=VAlign_Fill}}
  }
  WBP_ActionBarContainer {
   widgetClassPath=/Game/UI/Frontend/BottomBar/WBP_ActionBarContainer
   bIsVariable=true
   parent=Overlay_Content
   widget=WBP_ActionBarContainer
   slot {class=OverlaySlot, properties {horizontalAlignment=HAlign_Right, verticalAlignment=VAlign_Bottom}}
  }
  Slot_BottomBar {widgetClassPath=&NamedSlot, bIsVariable=true, namedSlotHost=WBP_ActionBarContainer, widget=Slot_BottomBar}
  Slot_ScreenOverlay {
   widgetClassPath=&NamedSlot
   bIsVariable=true
   parent=Overlay_34
   widget=Slot_ScreenOverlay
   slot {class=OverlaySlot, properties {horizontalAlignment=HAlign_Fill, verticalAlignment=VAlign_Fill}}
  }
 ]
}
